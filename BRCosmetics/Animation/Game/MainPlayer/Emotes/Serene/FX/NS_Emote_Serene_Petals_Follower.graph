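# Niagara effect for the Serene emote that spawns sparse, drifting petal sprites around the system origin.
# Petals spawn in a sphere, then gravity, turbulent wind and aerodynamic drag with lift and rotation make them tumble and bounce weakly on collision.
# Sprites follow a mesh orientation for facing and scale in and out over their short life.
# A GPU emitter is paired with a near-identical non-GPU LOW variant.
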NiagaraSystem /BRCosmetics/Animation/Game/MainPlayer/Emotes/Serene/FX/NS_Emote_Serene_Petals_Follower
dependencies {
 usedDataInterfaces [&NiagaraDataInterfaceCollisionQuery, &NiagaraDataInterfaceLandscape, &NiagaraDataInterfaceCurve]
 usedDynamicInputs []
 usedModules [&SystemState]
}
emitters {
 BlowingParticles {
  simTarget=GPUComputeSim
  emitterUpdateScript [
   &EmitterState()
   &SpawnRate(Spawn Probability=0.8, SpawnRate=&RandomRangeFloat(Minimum=1, Maximum=5), Use Spawn Probability=true)
  ]
  particleSpawnScript [
   &InitializeParticle(
    Lifetime Max=3
    Lifetime Min=2
    Lifetime Mode=::NewEnumerator1
    Mass Max=1.5
    Mass Mode=::NewEnumerator3
    Sprite Rotation Mode=::NewEnumerator2
    Sprite Size Mode=::NewEnumerator1
    Uniform Sprite Size Max=4
    Uniform Sprite Size Min=3
    UsePositionOffset=false
    Write Material Random=true
   )
   &ShapeLocation(MaskSpawnGroup=false, Override Seed=false, Sphere Radius=65, UseOwnerScale=true)
   &InitialMeshOrientation(Use Rotation Vector=false)
  ]
  particleUpdateScript [
   &ParticleState()
   &GravityForce(Gravity=<0 0 -750>)
   &WindForce(
    Friction=0.8
    Offset By Time:Vector3f=<50 10 -125>
    Scale Min / Max:Vector2f=<0.275 1>
    Speed Min / Max:Vector2f=<0 500>
    Turbulence Scale Mode:ENiagaraWindTurbulenceContributionMode=::NewEnumerator1
    Wind Speed=<-100 0 0>
    Wind Speed Scale=&RandomRangeFloat(Minimum=0.2, Maximum=0.7)
   )
   &AerodynamicDrag(
    Aerodynamic Drag:NiagaraFloat=&RandomRangeFloat(Minimum=0.3333, Maximum=1)
    Aerodynamic Rotational Drag=0.15
    Lift Contribution=0.75
    Pivot Offset Maximum:Vector3f=<5 0.5 1>
    Pivot Offset Minimum:Vector3f=<-5 -0.5 -1>
    Rotation Strength:NiagaraFloat=&RandomRangeFloat(Minimum=1.8, Maximum=3.2)
    Use Linear Drag=true
    Use Rotational Drag=true
   )
   &Collision(
    Age Colliding Particles=false
    Enable Rest State=false
    Friction=2
    Friction During a Bounce=0.1
    GPU Collision Type=::NewEnumerator2
    KillOnCollision=false
    Module.EnableMaxCollisionCount=false
    Module.Randomize Collision Normal Vector=false
    Particle Radius Scale=0.1
    Restitution=0.05
   )
   &SolveForcesAndVelocity(Clamp Velocity=false, Limit Acceleration=false)
   &ScaleSpriteSize(Uniform Curve Sprite Scale=DataInterface() {propertyValues {Curve {keys [
        {arriveTangent=14.5835, interpMode=RCIM_Cubic, leaveTangent=14.5835, tangentMode=RCTM_User, time=0, value=0}
        {arriveTangent=1.1111, interpMode=RCIM_Cubic, leaveTangent=1.1111, time=0.1, value=1}
        {arriveTangent=-1.1111, interpMode=RCIM_Cubic, leaveTangent=-1.1111, time=0.9, value=1}
        {arriveTangent=-8.7113, interpMode=RCIM_Cubic, leaveTangent=-8.7113, tangentMode=RCTM_User, time=1, value=0}
       ]}}})
   &AlignSpriteToMeshOrientation(SetAlignment=true, SetFacing=true)
   /Niagara/Modules/Collision/AlignParticlesWithCollisionPlane(Allow Alignment on Y=false, Allow Alignment on Z=false, Rotational Dampening Power=1) {enabled=false}
  ]
  rendererClasses [&NiagaraSpriteRendererProperties]
  renderers [{rendererClass=&NiagaraSpriteRendererProperties}]
 }
 BlowingParticlesLOW {
  emitterUpdateScript [
   &EmitterState()
   &SpawnRate(Spawn Probability=0.8, SpawnRate=&RandomRangeFloat(Minimum=1, Maximum=5), Use Spawn Probability=true)
  ]
  particleSpawnScript [
   &InitializeParticle(
    Lifetime Max=3
    Lifetime Min=2
    Lifetime Mode=::NewEnumerator1
    Mass Max=1.5
    Mass Mode=::NewEnumerator3
    Sprite Rotation Mode=::NewEnumerator2
    Sprite Size Mode=::NewEnumerator1
    Uniform Sprite Size Max=4
    Uniform Sprite Size Min=3
    UsePositionOffset=false
    Write Material Random=true
   )
   &ShapeLocation(MaskSpawnGroup=false, Override Seed=false, Sphere Radius=65, UseOwnerScale=true)
   &InitialMeshOrientation(Use Rotation Vector=false)
  ]
  particleUpdateScript [
   &ParticleState()
   &GravityForce(Gravity=<0 0 -750>)
   &WindForce(
    Offset By Time:Vector3f=<50 10 -125>
    Scale Min / Max:Vector2f=<0.275 1>
    Speed Min / Max:Vector2f=<0 500>
    Turbulence Scale Mode:ENiagaraWindTurbulenceContributionMode=::NewEnumerator1
    Wind Speed=<-100 0 0>
    Wind Speed Scale=&RandomRangeFloat(Minimum=0.2, Maximum=0.7)
   )
   &AerodynamicDrag(
    Aerodynamic Drag:NiagaraFloat=&RandomRangeFloat(Minimum=0.3333, Maximum=1)
    Aerodynamic Rotational Drag=0.15
    Lift Contribution=0.75
    Pivot Offset Maximum:Vector3f=<5 0.5 1>
    Pivot Offset Minimum:Vector3f=<-5 -0.5 -1>
    Rotation Strength:NiagaraFloat=&RandomRangeFloat(Minimum=1.8, Maximum=3.2)
    Use Linear Drag=true
    Use Rotational Drag=true
   )
   &Collision(
    Age Colliding Particles=false
    Enable Rest State=false
    Friction=2
    Friction During a Bounce=0.1
    KillOnCollision=false
    Module.EnableMaxCollisionCount=false
    Module.Randomize Collision Normal Vector=false
    Particle Radius Scale=0.1
    Restitution=0.05
   )
   &SolveForcesAndVelocity(Clamp Velocity=false, Limit Acceleration=false)
   &ScaleSpriteSize(Uniform Curve Sprite Scale=DataInterface() {propertyValues {Curve {keys [
        {arriveTangent=14.5835, interpMode=RCIM_Cubic, leaveTangent=14.5835, tangentMode=RCTM_User, time=0, value=0}
        {arriveTangent=1.1111, interpMode=RCIM_Cubic, leaveTangent=1.1111, time=0.1, value=1}
        {arriveTangent=-1.1111, interpMode=RCIM_Cubic, leaveTangent=-1.1111, time=0.9, value=1}
        {arriveTangent=-8.7113, interpMode=RCIM_Cubic, leaveTangent=-8.7113, tangentMode=RCTM_User, time=1, value=0}
       ]}}})
   &AlignSpriteToMeshOrientation(SetAlignment=true, SetFacing=true)
   &AlignParticlesWithCollisionPlane(Allow Alignment on Y=false, Allow Alignment on Z=false, Rotational Dampening Power=1) {enabled=false}
  ]
  rendererClasses [&NiagaraSpriteRendererProperties]
  renderers [{rendererClass=&NiagaraSpriteRendererProperties}]
 }
}
summary {
 systemName=NS_Emote_Serene_Petals_Follower
 emitters [
  {emitterName=BlowingParticles, simTarget=GPUComputeSim, rendererClasses [&NiagaraSpriteRendererProperties]}
  {emitterName=BlowingParticlesLOW, rendererClasses [&NiagaraSpriteRendererProperties]}
 ]
 userVariables [User.bIsFrontEndPreview {type=NiagaraBool}]
}
